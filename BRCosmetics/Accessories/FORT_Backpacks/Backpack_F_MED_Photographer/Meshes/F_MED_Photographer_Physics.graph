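# Physics asset for the Photographer backpack cosmetic, defining collision bodies and joints for the pack and its attached props.
# Box and capsule bodies cover the backpack, shield, sword, bottle, game console and five chain links.
# The sword hangs from the pack root, the shield, bottle and console branch off the backpack, and the chain trails from the bottle.
# Prop joints have per-axis angular limits, while the master-bone joints rotate freely.
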PhysicsAsset /BRCosmetics/Accessories/FORT_Backpacks/Backpack_F_MED_Photographer/Meshes/F_MED_Photographer_Physics
bodies {
 backpack [{shapeType=Box, extentX=18.0476, extentY=21.5122, extentZ=29.4208, center=<7.7489 -1.4137 2.807>, rotation=<rot 5 0 0>}]
 backpack_mstr [{shapeType=Capsule, length=0.505, radius=0.505}]
 bottle [{shapeType=Capsule, length=13.1823, radius=3.5488, center=<-1.8327 0.3828 -0.5183>, rotation=<rot 84.9738 -179.9991 -179.9991>}]
 chain1 [{shapeType=Box, extentX=1.475, extentY=0.6541, extentZ=1.8186, center=<0.7405 0.0077 -0.0053>}]
 chain2 [{shapeType=Box, extentX=1.51, extentY=2.01, extentZ=0.5, center=<0.7315 0.0057 0.0016>}]
 chain3 [{shapeType=Box, extentX=1.4712, extentY=0.6452, extentZ=1.8187, center=<0.6991 0.0018 -0.0054>}]
 chain4 [{shapeType=Box, extentX=1.51, extentY=2.01, extentZ=0.51, center=<0.7352 -0.0205 0.011>}]
 chain5 [{shapeType=Box, extentX=6.576, extentY=2.1516, extentZ=5.7048, center=<3.0303 -0.7304 0.106>}]
 gameConsole [{shapeType=Box, extentX=16.0158, extentY=9.5346, extentZ=2.4142, center=<-7.7961 0.1432 -0.8386>, rotation=<rot 4.9809 -5.0209 4.5632>}]
 pack [{shapeType=Capsule, length=0.505, radius=0.505}]
 shield [{shapeType=Box, extentX=3.5661, extentY=15.4673, extentZ=22.2047, center=<0 0.1 0.2>}]
 shield_mstr [{shapeType=Box, extentX=1.01, extentY=1.01, extentZ=1.01}]
 sword [{shapeType=Box, extentX=77.4891, extentY=5.3476, extentZ=21.4605, center=<-3.8008 -0.0076 -0.0551>}]
}
constraints [
 {bone1Name=sword, bone2Name=pack, swing1LimitDegrees=2, swing2LimitDegrees=45, twistLimitDegrees=5}
 {
  bone1Name=backpack_mstr
  bone2Name=pack
  swing1LimitDegrees=45
  swing1Motion=Free
  swing2LimitDegrees=45
  swing2Motion=Free
  twistLimitDegrees=45
  twistMotion=Free
 }
 {bone1Name=bottle, bone2Name=backpack, swing1LimitDegrees=2, swing2LimitDegrees=5, twistLimitDegrees=2}
 {bone1Name=shield, bone2Name=shield_mstr, swing1LimitDegrees=60, swing2LimitDegrees=10, twistLimitDegrees=45}
 {bone1Name=gameConsole, bone2Name=backpack, swing1LimitDegrees=15, swing2LimitDegrees=5, twistLimitDegrees=5}
 {
  bone1Name=backpack
  bone2Name=backpack_mstr
  swing1LimitDegrees=45
  swing1Motion=Free
  swing2LimitDegrees=45
  swing2Motion=Free
  twistLimitDegrees=45
  twistMotion=Free
 }
 {bone1Name=chain1, bone2Name=bottle, swing1LimitDegrees=30, swing2LimitDegrees=30, twistLimitDegrees=5}
 {bone1Name=chain2, bone2Name=chain1, swing1LimitDegrees=30, swing2LimitDegrees=30, twistLimitDegrees=5}
 {bone1Name=chain3, bone2Name=chain2, swing1LimitDegrees=30, swing2LimitDegrees=30, twistLimitDegrees=5}
 {bone1Name=chain4, bone2Name=chain3, swing1LimitDegrees=30, swing2LimitDegrees=30, twistLimitDegrees=5}
 {bone1Name=chain5, bone2Name=chain4, swing1LimitDegrees=30, swing2LimitDegrees=30, twistLimitDegrees=5}
 {
  bone1Name=shield_mstr
  bone2Name=backpack
  swing1LimitDegrees=45
  swing1Motion=Free
  swing2LimitDegrees=45
  swing2Motion=Free
  twistLimitDegrees=45
  twistMotion=Free
 }
]
